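; Emote helper Blueprint for the QuietPeanuts emote that spawns a camera lens effect for the locally viewed player.
; Shortly after BeginPlay it casts its attach parent to FortPawn and checks whether that pawn is locally viewed.
; If so, it spawns a generic lens effect from a class variable and stores the returned reference.
; EndPlay has no logic, so this graph performs no cleanup of the effect.
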(event EventBeginPlay
  (Utilities|FlowControl|Delay 0.1)
  (bind _asfort_pawn (Utilities|Casting|CastToFortPawn (Actor|GetAttachParentActor self)))
  (if (Pawn|IsLocallyViewed _asfort_pawn)
    (bind _returnvalue (Feedback|SpawnGenericCameraLensEffect _asfort_pawn (Variables|Default|GetLens%20Effect%20Emitter%20Class)))
    (Variables|Default|SetCameraLensFX _returnvalue)))

(event EventEndPlay
  (params (EndPlayReason :type EEndPlayReason%20Enum :default "Destroyed")))

(fn ConstructionScript)
