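# Niagara system for the QuietPeanuts emote that shows green goop droplets, apparently as a camera effect going by its name.
# A single GPU sprite emitter combines a randomized burst with a curve-decayed spawn rate over a brief loop duration, placing drops within a shape volume.
# Drops get no initial velocity or forces, shrink late in life, and fade via an alpha curve with randomized offset and scale.
# Dynamic material parameters feed randomized values to the sprite material.
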NiagaraSystem /BRCosmetics/Effects/Fort_Effects/Emotes/QuietPeanuts/NS_QuietPeanutsEmote_Goop_Drops_Camera
dependencies {usedDataInterfaces [&NiagaraDataInterfaceCurve], usedDynamicInputs [], usedModules [&SystemState]}
emitters {SimpleSpriteBurst {
  simTarget=GPUComputeSim
  emitterUpdateScript [
   &EmitterState(Life Cycle Mode=::NewEnumerator1, Loop Behavior=::NewEnumerator1, Loop Duration=0.3, UseLoopDelay=false)
   &SpawnBurst_Instantaneous(
    Spawn Count=&RandomRangeInt(Minimum=30, Maximum=50)
    Use Spawn Probability=false
    UseLoopCountLimit=false
   )
   &SpawnRate(
    SpawnRate=&FloatFromCurve(
     FloatCurve=DataInterface() {propertyValues {Curve {keys [{arriveTangent=-1.9873, interpMode=RCIM_Cubic, leaveTangent=-1.9873, tangentMode=RCTM_User, time=0, value=1}, <at 1 0>]}}}
     CurveIndex=$Emitter.NormalizedLoopAge
     Scale Curve=300
    )
    Use Spawn Probability=false
   )
  ]
  particleSpawnScript [
   &InitializeParticle(
    Color Maximum=<rgba 0.1016 0.5052 0 1>
    Color Minimum=<rgba 0.5313 1 0.003 1>
    Color Mode=::NewEnumerator2
    Lifetime Max=1
    Lifetime Min=2.2
    Lifetime Mode=::NewEnumerator1
    Sprite Size Mode=::NewEnumerator1
    Uniform Sprite Size Max=5
    Uniform Sprite Size Min=2
    UsePositionOffset=false
    Write Material Random=true
   )
   &ShapeLocation(MaskSpawnGroup=false, Override Seed=false, Shape Primitive=::NewEnumerator2, Surface Only=false, UseOwnerScale=true)
  ]
  particleUpdateScript [
   &ParticleState()
   &ScaleSpriteSize(Uniform Curve Sprite Scale=DataInterface() {propertyValues {Curve {keys [{tangentMode=RCTM_User, time=0.65, value=1}, {tangentMode=RCTM_User, time=1, value=0}]}}})
   &ScaleColor(
    Scale Alpha=&FloatFromCurve(
     FloatCurve=DataInterface() {propertyValues {Curve {keys [
         {arriveTangent=9.0725, interpMode=RCIM_Cubic, leaveTangent=9.0725, tangentMode=RCTM_User, time=0, value=0}
         {arriveTangentWeight=0.0249, interpMode=RCIM_Cubic, leaveTangentWeight=0.4235, tangentMode=RCTM_User, time=0.05, value=0.9847}
         {arriveTangentWeight=0.4081, interpMode=RCIM_Cubic, leaveTangentWeight=0.048, tangentMode=RCTM_User, time=0.95, value=1}
         {arriveTangent=-0.0405, interpMode=RCIM_Cubic, leaveTangent=-0.0405, tangentMode=RCTM_User, time=1, value=0}
        ]}}}
     CurveIndex=&Lerp_Float(A=&RandomRangeFloat(Minimum=0, Maximum=-0.2), B=1, Alpha=$Particles.NormalizedAge)
     Scale Curve=&RandomRangeFloat(Minimum=0.3, Maximum=0.5)
    )
    ScaleA=true
    ScaleRGB=false
   )
   &SolveForcesAndVelocity(Clamp Velocity=false, Limit Acceleration=false)
   &DynamicMaterialParameters(
    Index 0 Param 1=&FloatFromCurve(
     FloatCurve=DataInterface() {propertyValues {Curve {keys [
         {interpMode=RCIM_Cubic, tangentMode=RCTM_User, time=0, value=1}
         {interpMode=RCIM_Cubic, tangentMode=RCTM_User, time=0.2, value=0.9}
         {interpMode=RCIM_Cubic, tangentMode=RCTM_User, time=1, value=1}
        ]}}}
     Scale Curve=&RandomRangeFloat(Minimum=1, Maximum=1)
    )
    Index 0 Param 2=1
    Index 0 Param 3=1
    Index 0 Param 4=1
    Index 1 Param 1=&RandomRangeFloat(Minimum=0.5, Maximum=1)
    Index 1 Param 2=&RandomRangeFloat(Minimum=1, Maximum=2)
    Module.Param0WriteEnabled=true
    Module.Param0WriteEnabled1=true
    Module.Param1WriteEnabled=true
    Module.Param1WriteEnabled1=true
    Module.Param2WriteEnabled=true
    Module.Param2WriteEnabled1=true
    Module.Param3WriteEnabled=true
    Module.Param3WriteEnabled1=true
    Write Parameter Index 1=true
   )
  ]
  rendererClasses [&NiagaraSpriteRendererProperties]
  renderers [{rendererClass=&NiagaraSpriteRendererProperties}]
 }}
summary {
 systemName=NS_QuietPeanutsEmote_Goop_Drops_Camera
 emitters [{emitterName=SimpleSpriteBurst, simTarget=GPUComputeSim, rendererClasses [&NiagaraSpriteRendererProperties]}]
 userVariables []
}
