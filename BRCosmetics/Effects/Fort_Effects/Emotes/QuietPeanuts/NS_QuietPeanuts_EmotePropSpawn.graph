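# Niagara system for the prop spawn of the QuietPeanuts emote, a green goo splash with glow, smoke and sparkles.
# One-shot bursts drive the air glow, splash mesh, green light, smoke and curl-noise sparkles.
# Ring-spawned splash sprites launch outward under gravity; Splash_Hard adds collision and emits collision events tagged with a random VisibilityTag.
# A MedLow splash variant omits collision, mesh colour varies by platform, and Splash_01/02 are disabled.
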NiagaraSystem /BRCosmetics/Effects/Fort_Effects/Emotes/QuietPeanuts/NS_QuietPeanuts_EmotePropSpawn
dependencies {
 usedDataInterfaces [
  &NiagaraDataInterfaceCurve
  &NiagaraDataInterfaceArrayFloat3
  &NiagaraDataInterfaceVector2DCurve
  &NiagaraDataInterfaceArrayFloat
  &NiagaraDataInterfaceArrayFloat2
  &NiagaraDataInterfaceSpriteRendererInfo
  &NiagaraDataInterfacePlatformSet
  &NiagaraDataInterfaceVectorCurve
 ]
 usedDynamicInputs []
 usedModules [&SystemState]
}
emitters {
 AirGlow {
  emitterUpdateScript [
   &EmitterState()
   &SpawnBurst_Instantaneous(Spawn Count=1, Use Spawn Probability=false)
  ]
  particleSpawnScript [
   &InitializeParticle(
    Color=<rgba 0.4323 0.4323 0.4323 0.2>
    Lifetime=0.25
    Position Offset=<0 0 25>
    Sprite Size Mode=::NewEnumerator3
    Uniform Sprite Size=300
    UsePositionOffset=true
    Write Material Random=true
   ) {module=InitializeParticle001}
   &DynamicMaterialParameters(
    Index 0 Param 1=2
    Module.Param0WriteEnabled=true
    Module.Param1WriteEnabled=true
    Module.Param2WriteEnabled=true
    Module.Param3WriteEnabled=true
   )
   &CameraOffset(Camera Offset Amount=10)
  ]
  particleUpdateScript [
   &ParticleState()
   &ScaleSpriteSize(Scale Factor=&Vector2DFromFloat(Value=&FloatFromCurve(FloatCurve=DataInterface() {propertyValues {Curve {keys [<at 0 1 RCIM_Cubic>]}}})))
   &ScaleColor(
    Scale Alpha=&FloatFromCurve(FloatCurve=DataInterface() {propertyValues {Curve {keys [<at 0 0>,<at 0.25 0.75>,<at 1 0>]}}})
    ScaleA=true
    ScaleRGB=true
   ) {module=ScaleColor001}
  ]
  rendererClasses [&NiagaraSpriteRendererProperties, &NiagaraMeshRendererProperties]
  renderers [
   {rendererClass=&NiagaraSpriteRendererProperties}
   {rendererClass=&NiagaraMeshRendererProperties, rendererIndex=1}
  ]
 }
 BuffSplash_Mesh {
  emitterUpdateScript [
   &EmitterState()
   &SpawnBurst_Instantaneous(Spawn Count=4, Use Spawn Probability=false)
  ]
  particleSpawnScript [
   &InitializeParticle(
    Color=&MakeCustomLinearColorFromBool(
     Bool=&IsPlatformSetActive(New Platform Set=DataInterface() {propertyValues {Platforms {qualityLevelMask=-29}}})
     True Color=&RandomRangeLinearColor(Minimum=<rgba 0.4905 1 0 1>, Maximum=<rgba 0.1623 1 0 1>)
     False Color=&RandomRangeLinearColor(Minimum=<rgba 0.3186 0.5087 0 1>, Maximum=<rgba 0.1309 0.4566 0 1>)
    )
    Mesh Scale Max=<32 32 20>
    Mesh Scale Min=<16 16 10>
    Mesh Scale Mode=::NewEnumerator2
    UsePositionOffset=true
    Write Material Random=true
   )
   &DynamicMaterialParameters(
    Index 0 Param 1=0.5
    Index 0 Param 2=0.5
    Index 0 Param 3=25
    Index 0 Param 4=&RandomRangeFloat(Minimum=0, Maximum=10)
    Module.Param0WriteEnabled=true
    Module.Param1WriteEnabled=true
    Module.Param2WriteEnabled=true
    Module.Param3WriteEnabled=true
   ) {module=DynamicMaterialParameters001}
   &InitialMeshOrientation(
    Rotation=&RandomRangeVector(Minimum=<-0.05 -0.05 0>, Maximum=<0.05 0.05 1>)
    Use Orientation Vector=true
    Use Rotation Vector=true
   )
  ]
  particleUpdateScript [
   &ParticleState()
   /SetVariables() {enabled=false, isSetParameters=true}
   &ScaleColor(
    Scale Alpha=&FloatFromCurve(FloatCurve=DataInterface() {propertyValues {Curve {keys [
         {arriveTangent=-0.6531, interpMode=RCIM_Cubic, time=0, value=1}
         {arriveTangent=-0.8638, interpMode=RCIM_Cubic, leaveTangent=-0.8638, tangentMode=RCTM_User, time=0.8, value=0.3}
         {arriveTangent=-0.4827, interpMode=RCIM_Cubic, leaveTangent=-0.4827, tangentMode=RCTM_User, time=1, value=0.1}
        ]}}})
    ScaleA=true
    ScaleRGB=false
   )
   &ScaleMeshSize(Scale Factor=&VectorFromCurve(
     VectorCurve=DataInterface() {propertyValues {
       XCurve {keys [
         {arriveTangent=1.8221, interpMode=RCIM_Cubic, leaveTangent=1.8221, tangentMode=RCTM_User, time=0, value=0}
         {arriveTangent=0.4715, interpMode=RCIM_Cubic, leaveTangent=0.4715, tangentMode=RCTM_User, time=1, value=1}
        ]}
       YCurve {keys [
         {arriveTangent=1.8221, interpMode=RCIM_Cubic, leaveTangent=1.8221, tangentMode=RCTM_User, time=0, value=0}
         {arriveTangent=0.4715, interpMode=RCIM_Cubic, leaveTangent=0.4715, tangentMode=RCTM_User, time=1, value=1}
        ]}
       ZCurve {keys [
         {arriveTangent=4.7636, interpMode=RCIM_Cubic, leaveTangent=4.7636, tangentMode=RCTM_User, time=0, value=0}
         {arriveTangentWeight=0.2848, interpMode=RCIM_Cubic, leaveTangentWeight=0.1221, tangentMode=RCTM_User, time=0.5, value=1}
         {arriveTangent=-5.505, interpMode=RCIM_Cubic, leaveTangent=-5.505, tangentMode=RCTM_User, time=1, value=0}
        ]}
      }}
     Scale Curve=&VectorFromFloat(Value=0.16)
    )) {enabled=false}
   &ScaleMeshSize(Scale Factor=&VectorFromCurve(
     VectorCurve=DataInterface() {propertyValues {
       XCurve {keys [
         {arriveTangent=1.8221, interpMode=RCIM_Cubic, leaveTangent=1.8221, tangentMode=RCTM_User, time=0, value=0}
         {arriveTangent=0.4715, interpMode=RCIM_Cubic, leaveTangent=0.4715, tangentMode=RCTM_User, time=1, value=1}
        ]}
       YCurve {keys [
         {arriveTangent=1.8221, interpMode=RCIM_Cubic, leaveTangent=1.8221, tangentMode=RCTM_User, time=0, value=0}
         {arriveTangent=0.4715, interpMode=RCIM_Cubic, leaveTangent=0.4715, tangentMode=RCTM_User, time=1, value=1}
        ]}
       ZCurve {keys [
         {arriveTangent=4.7636, interpMode=RCIM_Cubic, leaveTangent=4.7636, tangentMode=RCTM_User, time=0, value=0}
         {arriveTangentWeight=0.2848, interpMode=RCIM_Cubic, leaveTangentWeight=0.1221, tangentMode=RCTM_User, time=0.5, value=1}
         {arriveTangent=-5.505, interpMode=RCIM_Cubic, leaveTangent=-5.505, tangentMode=RCTM_User, time=1, value=0}
        ]}
      }}
     Scale Curve=&VectorFromFloat(Value=0.14)
    )) {module=ScaleMeshSize001}
   &DynamicMaterialParameters(
    Index 1 Param 2=&FloatFromCurve(
     FloatCurve=DataInterface() {propertyValues {Curve {keys [<at 0 -1.5 RCIM_Cubic>, {arriveTangent=1.2375, interpMode=RCIM_Cubic, leaveTangent=1.2375, tangentMode=RCTM_User, time=1, value=1}]}}}
     Scale Curve=-20
    )
    Module.Param0WriteEnabled1=false
    Module.Param1WriteEnabled1=true
    Module.Param2WriteEnabled1=false
    Module.Param3WriteEnabled1=false
    Write Parameter Index 0=false
    Write Parameter Index 1=true
   )
  ]
  rendererClasses [&NiagaraMeshRendererProperties]
  renderers [{rendererClass=&NiagaraMeshRendererProperties, rendererIndex=0}]
 }
 Light {
  emitterUpdateScript [
   &EmitterState()
   &SpawnBurst_Instantaneous(Use Spawn Probability=false, UseLoopCountLimit=false)
  ]
  particleSpawnScript [&InitializeParticle(
    Color=<rgba 0.2621 0.9823 0 40>
    Lifetime=3
    Position Offset=&FN_FrontEndVector3Modifier(Gameplay=<0 0 10>, Front End=<0 0 -1>)
    UsePositionOffset=true
    Write Material Random=true
   )]
  particleUpdateScript [
   &ParticleState()
   &FN_Light_Attributes(Create Exponent Attribute=true, Create Radius Attribute=true, Exponent=3, Radius=300)
   &ScaleColor(
    Scale Alpha=&FloatFromCurve(FloatCurve=DataInterface() {propertyValues {Curve {keys [
         {arriveTangent=9.0725, interpMode=RCIM_Cubic, leaveTangent=9.0725, tangentMode=RCTM_User, time=0, value=0}
         {arriveTangent=-2.5249, interpMode=RCIM_Cubic, leaveTangent=-2.5249, tangentMode=RCTM_User, time=0.1408, value=0.9847}
         <at 0.4 0.4>
         <at 0.6 0.4>
         {arriveTangent=-0.0405, interpMode=RCIM_Cubic, leaveTangent=-0.0405, tangentMode=RCTM_User, time=1, value=0}
        ]}}})
    ScaleA=true
    ScaleRGB=true
   )
  ]
  rendererClasses [&NiagaraLightRendererProperties]
  renderers [{rendererClass=&NiagaraLightRendererProperties, rendererIndex=0}]
 }
 Smoke {
  emitterUpdateScript [
   &EmitterState()
   &SpawnBurst_Instantaneous(Spawn Count=7, Use Spawn Probability=false) {module=SpawnBurst_Instantaneous001}
  ]
  particleSpawnScript [
   &InitializeParticle(
    Color=<rgba 0.5202 0.5255 0.5202 1>
    Lifetime=&UniformRangedFloat(Minimum=0.2, Maximum=0.6)
    Sprite Rotation Angle=&UniformRangedFloat(Minimum=0, Maximum=1)
    Sprite Rotation Mode=::NewEnumerator1
    Sprite Size Mode=::NewEnumerator3
    Uniform Sprite Size=&UniformRangedFloat(Minimum=50, Maximum=35)
    UsePositionOffset=false
    Write Material Random=true
   ) {module=InitializeParticle001}
   /SetVariables(Particles.InitSpriteRotRate=&UniformRangedFloat(Minimum=-1, Maximum=1)) {isSetParameters=true}
   &AccelerationForce(Acceleration=&UniformRangedVector(Minimum=<0 0 -25>, Maximum=<0 0 -75>))
   &SphereLocation(
    Hemisphere Z=true
    MaskSpawnGroup=false
    Override Seed=false
    Sphere Radius=10
    Surface Only=true
    UseNonUniformScale=false
    UseOffset=true
    UseSphereOrientation=false
   )
   &ApplyInitialForces()
   &CameraOffset(Camera Offset Amount=&RandomRangeFloat(Minimum=0, Maximum=15))
   &AddVelocity(Velocity=&Multiply_Vector(
     A=$Output.SphereLocation.SphereNormal
     B=&VectorFromFloat(Value=&RandomRangeFloat(Minimum=500, Maximum=300))
    ))
  ]
  particleUpdateScript [
   &ParticleState()
   &SpriteRotationRate(Rotation Rate=&Multiply_Float(
     A=$Particles.InitSpriteRotRate
     B=&FloatFromCurve(
      FloatCurve=DataInterface() {propertyValues {Curve {keys [
          {arriveTangent=-1.9214, interpMode=RCIM_Cubic, leaveTangent=-1.9214, tangentMode=RCTM_User, time=-0.0046, value=1.0113}
          {arriveTangent=-0.2496, interpMode=RCIM_Cubic, leaveTangent=-0.2496, tangentMode=RCTM_User, time=0.9896, value=0.16}
         ]}}}
      CurveIndex=$Emitter.Age
     )
    ))
   &ScaleColor(
    Scale Alpha=&FloatFromCurve(FloatCurve=DataInterface() {propertyValues {Curve {keys [<at 0 0>,<at 0.2049 0.9765>,<at 0.7791 0.9107>,<at 1 0>]}}})
    ScaleA=true
    ScaleRGB=true
   ) {module=ScaleColor001}
   &Drag(Drag=3, Use Linear Drag=true, Use Rotational Drag=false)
   &SolveForcesAndVelocity(Clamp Velocity=false, Limit Acceleration=false)
   &ScaleSpriteSize(Scale Factor=&Vector2DFromFloat(Value=&FloatFromCurve(FloatCurve=DataInterface() {propertyValues {Curve {keys [
          {arriveTangent=5.2354, interpMode=RCIM_Cubic, leaveTangent=5.2354, tangentMode=RCTM_User, time=-0.0046, value=0.25}
          <at 0.1439 1.4923>
          <at 1.0011 3.4809 RCIM_Cubic>
         ]}}})))
   &DynamicMaterialParameters(
    Index 0 Param 1=&FloatFromCurve(FloatCurve=DataInterface() {propertyValues {Curve {keys [<at 0 0 RCIM_Cubic>,<at 1.0011 0.2>]}}})
    Index 0 Param 2=&FloatFromCurve(FloatCurve=DataInterface() {propertyValues {Curve {keys [
         {arriveTangent=0.0453, interpMode=RCIM_Cubic, tangentMode=RCTM_Break, time=0, value=-1}
         {arriveTangent=0.0604, interpMode=RCIM_Cubic, leaveTangent=0.0604, tangentMode=RCTM_User, time=1, value=-1}
        ]}}})
    Index 0 Param 3=&FloatFromCurve(FloatCurve=DataInterface() {propertyValues {Curve {keys [{interpMode=RCIM_Cubic, leaveTangent=4.6779, time=0.2851, value=0}]}}})
    Index 1 Param 1=&FloatFromCurve(
     FloatCurve=DataInterface() {propertyValues {Curve {keys [<at 0 0>,<at 0.25 1>]}}}
     Scale Curve=&RandomRangeFloat(Minimum=2, Maximum=1)
    )
    Index 1 Param 2=&FloatFromCurve(
     FloatCurve=DataInterface() {propertyValues {Curve {keys [
         {interpMode=RCIM_Cubic, leaveTangent=3.7422, tangentMode=RCTM_Break, time=0, value=0.2}
         {interpMode=RCIM_Cubic, leaveTangent=0.9066, time=1, value=2}
        ]}}}
     Scale Curve=&RandomRangeFloat(Minimum=0.5, Maximum=1)
    )
    Index 1 Param 3=&FloatFromCurve(FloatCurve=DataInterface() {propertyValues {Curve {keys [
         {interpMode=RCIM_Cubic, leaveTangent=2.8445, tangentMode=RCTM_Break, time=0, value=0.2}
         {interpMode=RCIM_Cubic, leaveTangent=63.2978, time=0.25, value=1}
        ]}}})
    Index 1 Param 4=&FloatFromCurve(
     FloatCurve=DataInterface() {propertyValues {Curve {keys [<at 0 1>,<at 1 0.8>]}}}
     Scale Curve=&RandomRangeFloat(Minimum=0.5, Maximum=1)
    )
    Index 2 Param 1=&FloatFromCurve(FloatCurve=DataInterface() {propertyValues {Curve {keys [{interpMode=RCIM_Cubic, leaveTangent=3.2309, tangentMode=RCTM_Break, time=0.1, value=0}, <at 1 1>]}}})
    Index 2 Param 2=$Particles.MaterialRandom
    Index 2 Param 3=&FloatFromCurve(FloatCurve=DataInterface() {propertyValues {Curve {keys [
         {interpMode=RCIM_Cubic, tangentMode=RCTM_Break, time=0, value=1}
         {interpMode=RCIM_Cubic, leaveTangent=11.1233, tangentMode=RCTM_Break, time=1, value=2}
        ]}}})
    Index 2 Param 4=0.75
    Module.Param0WriteEnabled=true
    Module.Param0WriteEnabled1=true
    Module.Param0WriteEnabled2=true
    Module.Param1WriteEnabled=false
    Module.Param1WriteEnabled1=true
    Module.Param1WriteEnabled2=true
    Module.Param2WriteEnabled=true
    Module.Param2WriteEnabled1=true
    Module.Param2WriteEnabled2=true
    Module.Param3WriteEnabled=true
    Module.Param3WriteEnabled1=true
    Module.Param3WriteEnabled2=true
    Write Parameter Index 1=true
    Write Parameter Index 2=true
   )
  ]
  rendererClasses [&NiagaraSpriteRendererProperties]
  renderers [{rendererClass=&NiagaraSpriteRendererProperties}]
 }
 Sparkles {
  emitterUpdateScript [
   &EmitterState()
   &SpawnBurst_Instantaneous(Spawn Count=13, Use Spawn Probability=false)
  ]
  particleSpawnScript [
   &InitializeParticle(
    Color=<rgba 5 3.7043 1.51 1>
    Lifetime=&UniformRangedFloat(Minimum=0.3, Maximum=1)
    Sprite Size Mode=::NewEnumerator3
    Uniform Sprite Size=&UniformRangedFloat(Minimum=10, Maximum=30)
    UsePositionOffset=false
    Write Material Random=true
   ) {module=InitializeParticle001}
   &AddVelocity(Velocity=&UniformRangedVector(Minimum=<-200 -200 -200>, Maximum=<200 200 200>))
   &ApplyInitialForces()
  ]
  particleUpdateScript [
   &ParticleState()
   &GravityForce(Gravity=<0 0 200>)
   /Niagara/Modules/Update/Forces/CurlNoiseForce(Noise Frequency=0.02, Noise Quality / Cost=::NewEnumerator0, Noise Strength=500, Pan Noise=false)
   &Drag(Drag=5, Use Linear Drag=true, Use Rotational Drag=false)
   &SolveForcesAndVelocity(Clamp Velocity=false, Limit Acceleration=false)
   &ScaleColor(
    Scale Alpha=&FloatFromCurve(FloatCurve=DataInterface() {propertyValues {Curve {keys [<at -0.0092 0.0144>,<at 0.0967 0.9767>,<at 1.0011 0.002>]}}})
    ScaleA=true
    ScaleRGB=true
   )
   &ScaleSpriteSize(Scale Factor=&Vector2DFromFloat(Value=&FloatFromCurve(FloatCurve=DataInterface() {propertyValues {Curve {keys [<at -0.0046 0>,<at 0.1746 0.9952>,<at 1.0011 0.002>]}}})))
  ]
  rendererClasses [&NiagaraSpriteRendererProperties]
  renderers [{rendererClass=&NiagaraSpriteRendererProperties}]
 }
 Splash_01 {
  bEnabled=false
  emitterUpdateScript [
   &EmitterState(Life Cycle Mode=::NewEnumerator1, Loop Behavior=::NewEnumerator1, Loop Duration=0.8, UseLoopDelay=false)
   &SpawnBurst_Instantaneous(Spawn Count=6, Use Spawn Probability=false) {module=SpawnBurst_Instantaneous001}
   &SpawnRate(
    SpawnRate=&FloatFromCurve(
     FloatCurve=DataInterface() {propertyValues {Curve {keys [
         {arriveTangent=-2.8421, interpMode=RCIM_Cubic, leaveTangent=-2.8421, tangentMode=RCTM_User, time=0, value=1}
         {interpMode=RCIM_Cubic, tangentMode=RCTM_User, time=0.9, value=0}
        ]}}}
     CurveIndex=$Emitter.NormalizedLoopAge
     Scale Curve=25
    )
    Use Spawn Probability=false
   )
   &SpawnBurst_Instantaneous(Spawn Count=24, Spawn Time=0.1, Use Spawn Probability=false) {enabled=false}
   &SpawnBurst_Instantaneous(Spawn Count=12, Spawn Time=0.15, Use Spawn Probability=false) {enabled=false, module=SpawnBurst_Instantaneous002}
  ]
  particleSpawnScript [
   &FN_SetSpawnGroup()
   /Niagara/Modules/Spawn/Initialization/InitializeParticle(
    Color=&RandomRangeLinearColor(Minimum=<rgba 3.2063 4.2281 5 2>, Maximum=<rgba 0.2 0.25 0.3 0>)
    Lifetime=&RandomRangeFloat(Minimum=0.8, Maximum=1.4)
    Mass=&RandomRangeFloat(Minimum=1, Maximum=2)
    Module.Write Color=true
    Module.Write Lifetime=true
    Module.Write Mass=true
    Module.Write Material Random=true
    Module.Write Position=false
    Module.Write Scale=false
    Module.Write SpriteRotation=true
    Module.Write SpriteSize=true
    Position=&SimulationPosition()
    Sprite Rotation=&UniformRangedFloat(Minimum=0, Maximum=360)
    Sprite Size=&Vector2DFromFloat(Value=&UniformRangedFloat(Minimum=10, Maximum=30))
   ) {enabled=false}
   /Niagara/Modules/Spawn/Initialization/V2/InitializeParticle(
    Color=&RandomRangeLinearColor(Minimum=<rgba 0.05 3 0 1>, Maximum=<rgba 2.1115 5 0 1>)
    Lifetime=&RandomRangeFloat(Minimum=0.5, Maximum=0.7)
    Mass=&RandomRangeFloat(Minimum=1, Maximum=2)
    Mass Mode=::NewEnumerator1
    Position Offset=$System.Location Offset
    Sprite Rotation Mode=::NewEnumerator2
    Sprite Size Max=<45 45>
    Sprite Size Min=<30 30>
    Sprite Size Mode=::NewEnumerator2
    Sprite UV Mode=::NewEnumerator3
    UsePositionOffset=true
    Write Material Random=true
   ) {module=InitializeParticle001}
   &ShapeLocation(MaskSpawnGroup=false, Offset Mode=::NewEnumerator0, Override Seed=false, Sphere Radius=50, UseOwnerScale=true)
   &AddVelocity(
    Module.Use Falloff Distance=false
    Override Seed=false
    Velocity Mode=::NewEnumerator1
    Velocity Speed=&RandomRangeFloat(Minimum=0, Maximum=100)
   )
   &AddVelocityFromPoint(
    Module.Use Falloff Distance=false
    Origin Offset=&SelectVectorFromArray(
     Direct Array Index=$Particles.SpawnGroup
     Vector Selection Array=DataInterface() {propertyValues {FloatData [<-5 0 -12>, <5 0 -12>, <0 -5 -12>, <0 5 -12>]}}
    )
    Origin Offset Coordinate Space=::World
    Velocity Strength=&UniformRangedFloat(Minimum=450, Maximum=800)
   ) {enabled=false}
  ]
  particleUpdateScript [
   &ParticleState()
   &AccelerationForce(Acceleration=&UniformRangedVector(Minimum=<0 0 -300>, Maximum=<0 0 -500>))
   &Drag(Use Linear Drag=true, Use Rotational Drag=true)
   /Niagara/Modules/Update/Forces/V2/CurlNoiseForce(Noise Strength=100, Pan Noise=true, Pan Noise Field=<0 0 1>)
   &SolveForcesAndVelocity(Clamp Velocity=false, Limit Acceleration=false)
   &ScaleSpriteSize(Scale Factor=&Vector2DFromFloat(Value=&FloatFromCurve(
      FloatCurve=DataInterface() {propertyValues {Curve {keys [
          {arriveTangent=12.5197, interpMode=RCIM_Cubic, leaveTangent=12.5197, tangentMode=RCTM_User, time=0, value=0.4376}
          {arriveTangent=0.8929, interpMode=RCIM_Cubic, leaveTangent=0.8929, tangentMode=RCTM_User, time=1, value=5}
         ]}}}
      Scale Curve=&RandomRangeFloat(Minimum=1, Maximum=1.5)
     )))
   &ScaleColor(
    Scale Alpha=&FloatFromCurve(FloatCurve=DataInterface() {propertyValues {Curve {keys [
         {interpMode=RCIM_Cubic, tangentMode=RCTM_User, time=0, value=0}
         {interpMode=RCIM_Cubic, tangentMode=RCTM_User, time=0.2, value=1}
         {tangentMode=RCTM_User, time=0.7, value=0}
        ]}}})
    ScaleA=true
    ScaleRGB=true
   )
   &DynamicMaterialParameters(
    Index 0 Param 1=&FloatFromCurve(FloatCurve=DataInterface() {propertyValues {Curve {keys [
         {arriveTangent=-2.0663, interpMode=RCIM_Cubic, leaveTangent=-2.0663, tangentMode=RCTM_User, time=0.25, value=1}
         {arriveTangent=-0.2741, interpMode=RCIM_Cubic, leaveTangent=-0.2741, tangentMode=RCTM_User, time=1.0011, value=0}
        ]}}})
    Index 0 Param 2=&FloatFromCurve(FloatCurve=DataInterface() {propertyValues {Curve {keys [<at -0.0086 -1.3001>,<at 0.9945 1.9153>]}}})
    Index 0 Param 3=0.5
    Module.Param0WriteEnabled=true
    Module.Param1WriteEnabled=false
    Module.Param2WriteEnabled=true
    Module.Param3WriteEnabled=true
   ) {module=DynamicMaterialParameters001}
   &SpriteRotationRate(Rotation Rate=&Lerp_Float(A=-30, B=30, Alpha=&RandomRangeFloat(Minimum=0, Maximum=1)))
  ]
  rendererClasses [&NiagaraSpriteRendererProperties]
  renderers [{rendererClass=&NiagaraSpriteRendererProperties}]
 }
 Splash_02 {
  bEnabled=false
  emitterUpdateScript [
   &EmitterState(Life Cycle Mode=::NewEnumerator1, Loop Behavior=::NewEnumerator1, Loop Delay=0.1, Loop Duration=0.8, UseLoopDelay=true)
   &SpawnBurst_Instantaneous(Spawn Count=6, Use Spawn Probability=false) {module=SpawnBurst_Instantaneous001}
   &SpawnRate(
    SpawnRate=&FloatFromCurve(
     FloatCurve=DataInterface() {propertyValues {Curve {keys [
         {arriveTangent=-2.8421, interpMode=RCIM_Cubic, leaveTangent=-2.8421, tangentMode=RCTM_User, time=0, value=1}
         {interpMode=RCIM_Cubic, tangentMode=RCTM_User, time=0.9, value=0}
        ]}}}
     CurveIndex=$Emitter.NormalizedLoopAge
     Scale Curve=25
    )
    Use Spawn Probability=true
   )
  ]
  particleSpawnScript [
   &InitializeParticle(
    Color=&RandomRangeLinearColor(Minimum=<rgba 0 5 0.0388 1>, Maximum=<rgba 2.235 5 0 1>)
    Lifetime Max=0.3
    Lifetime Min=0.1
    Lifetime Mode=::NewEnumerator1
    Position Offset=$System.Location Offset
    Position Offset Coordinate Space=::Local
    Sprite Rotation Mode=::NewEnumerator2
    Sprite Size Max=<100 100>
    Sprite Size Min=<80 80>
    Sprite Size Mode=::NewEnumerator2
    Sprite UV Mode=::NewEnumerator1
    UsePositionOffset=true
    Write Material Random=true
   )
   &SubUVAnimation(End Frame=3, SubUV Animation Mode:ENiagara_SubUVLookupMode=::NewEnumerator2)
   &ShapeLocation(MaskSpawnGroup=false, Override Seed=false, Sphere Radius=60, UseOwnerScale=true)
   &AddVelocity(
    Module.Use Falloff Distance=false
    Override Seed=false
    Velocity Mode=::NewEnumerator1
    Velocity Speed=&RandomRangeFloat(Minimum=100, Maximum=200)
   ) {module=AddVelocity001}
  ]
  particleUpdateScript [
   &ParticleState()
   &CameraOffset(Camera Offset Amount=&Multiply_Float(A=&MakeFloatFromVector2d(Vector 2D=$Particles.SpriteSize), B=0.33))
   &GravityForce(Gravity=&Multiply_VectorByFloat(Vector=<0 0 -980>, Float=0.2))
   &Drag(
    Drag=&RandomRangeFloat(Minimum=2, Maximum=3)
    Rotational Drag=50
    Use Linear Drag=true
    Use Rotational Drag=false
   )
   &InheritSourceMovement(
    Applied Position Scale=&VectorFromFloat(Value=&FloatFromCurve(
      FloatCurve=DataInterface() {propertyValues {Curve {keys [{interpMode=RCIM_Cubic, tangentMode=RCTM_User, time=0, value=1}, {interpMode=RCIM_Cubic, tangentMode=RCTM_User, time=0.66, value=0}]}}}
      Scale Curve=$System.Inherit Source Movement
     ))
    ApplyForce=false
    ApplyPosition=true
    ApplyPositionFalloff=false
   ) {enabled=false}
   &SolveForcesAndVelocity(Clamp Velocity=false, Limit Acceleration=false)
   &SpriteRotationRate(Rotation Rate=&FloatFromCurve(
     FloatCurve=DataInterface() {propertyValues {Curve {keys [<at 0 1>,<at 1 0>]}}}
     Scale Curve=&MakeCustomFloatFromBool(Bool=&RandomBool(), True Float=-90, False Float=90)
    ))
   &ScaleSpriteSize(Scale Factor=&Vector2DFromCurve(
     Vector2Curve=DataInterface() {propertyValues {
       XCurve {keys [{tangentMode=RCTM_User, time=0, value=0.4}, {tangentMode=RCTM_User, time=1, value=1}]}
       YCurve {keys [
         {arriveTangent=1.4008, interpMode=RCIM_Cubic, leaveTangent=1.4008, tangentMode=RCTM_User, time=0, value=0.25}
         {interpMode=RCIM_Cubic, tangentMode=RCTM_User, time=1, value=1}
        ]}
      }}
     ScaleCurve=<1 1.2>
    ))
   &ScaleColor(
    Scale Alpha=&FloatFromCurve(FloatCurve=DataInterface() {propertyValues {Curve {keys [{interpMode=RCIM_Cubic, tangentMode=RCTM_User, time=0.6, value=1}, {interpMode=RCIM_Cubic, tangentMode=RCTM_User, time=1, value=0}]}}})
    ScaleA=true
    ScaleRGB=true
   )
   &DynamicMaterialParameters(
    Index 0 Param 2=&FloatFromCurve(
     FloatCurve=DataInterface() {propertyValues {Curve {keys [{arriveTangent=-1.9579, interpMode=RCIM_Cubic, time=0.2, value=1}, {interpMode=RCIM_Cubic, tangentMode=RCTM_User, time=0.45, value=0}]}}}
     Scale Curve=0.5
    )
    Index 0 Param 4=&FloatFromCurve(
     FloatCurve=DataInterface() {propertyValues {Curve {keys [{interpMode=RCIM_Cubic, tangentMode=RCTM_User, time=0, value=0}, {interpMode=RCIM_Cubic, tangentMode=RCTM_User, time=1, value=1}]}}}
     Scale Curve=&RandomRangeFloat(Minimum=0.5, Maximum=1)
    )
    Module.Param0WriteEnabled=false
    Module.Param1WriteEnabled=true
    Module.Param2WriteEnabled=false
    Module.Param3WriteEnabled=true
   )
  ]
  rendererClasses [&NiagaraSpriteRendererProperties]
  renderers [{rendererClass=&NiagaraSpriteRendererProperties}]
 }
 Splash_Hard {
  emitterUpdateScript [
   &EmitterState(Life Cycle Mode=::NewEnumerator1, Loop Behavior=::NewEnumerator1, Loop Duration=0.15, UseLoopDelay=true)
   &SpawnBurst_Instantaneous(Spawn Count=7, Use Spawn Probability=false, UseLoopCountLimit=false) {module=SpawnBurst_Instantaneous001}
   &SpawnRate(SpawnGroup=1, SpawnRate=75, Use Spawn Probability=true)
  ]
  particleSpawnScript [
   &InitializeParticle(
    Color=&RandomRangeLinearColor(Minimum=<rgba 1 1 1 1>, Maximum=<rgba 0.138 0.6337 0.1353 1>)
    Lifetime Min=2
    Lifetime Mode=::NewEnumerator1
    Position Offset Coordinate Space=::Local
    Sprite Rotation Mode=::NewEnumerator2
    Sprite Size Max=<100 150>
    Sprite Size Min=<80 200>
    Sprite Size Mode=::NewEnumerator2
    Sprite UV Mode=::NewEnumerator3
    UsePositionOffset=true
    Write Material Random=true
   )
   &PartitionParticles(Number Of Partitions=1, Particles Per Partition=7, Partition Mode:ENiagara_PartitionMode=::NewEnumerator1)
   &SubUVAnimation(End Frame=3, SubUV Animation Mode:ENiagara_SubUVLookupMode=::NewEnumerator2)
   &ShapeLocation(
    MaskSpawnGroup=false
    Override Seed=false
    Ring / Disc Distribution Mode:ENiagaraSphereDistributionMode=::NewEnumerator1
    Ring Radius=30
    Shape Primitive=::NewEnumerator4
    U Position=&MaskFloatBySpawnGroup(
     Masked Float=$Particles.PartitionParticles.Normalized Index In Partition
     Passthrough Float=&RandomRangeFloat(Minimum=0, Maximum=1)
    )
    UseOwnerScale=true
   )
   /SetVariables(Particles.VisibilityTag=&MakeCustomIntFromBool(
     Bool=&SetBoolByFloatComparison(A=&RandomRangeFloat(Minimum=0, Maximum=1), B=0.6)
     True INT=10
     False INT=&RandomRangeInt(Minimum=0, Maximum=1)
    )) {isSetParameters=true, module=SetVariables_#1}
   &AddVelocity(
    Module.Use Falloff Distance=false
    Origin Offset=<0 0 -80>
    Override Seed=false
    Velocity Mode=::NewEnumerator1
    Velocity Speed=&Multiply_Float(
     A=&RandomRangeFloat(Minimum=500, Maximum=600)
     B=&SelectFloatFromArray(Direct Array Index=$Particles.VisibilityTag, Float Selection Array=DataInterface() {propertyValues {FloatData [1.25,1.25,0.8]}})
    )
   ) {enabled=false, module=AddVelocity001}
   &AddVelocity(
    Module.Use Falloff Distance=false
    Origin Offset=<0 0 -80>
    Override Seed=false
    Velocity Mode=::NewEnumerator1
    Velocity Speed=&Multiply_Float(A=&RandomRangeFloat(Minimum=500, Maximum=600), B=1.5)
   )
   /SetVariables() {isSetParameters=true, module=SetVariables_#2}
   /SetVariables(Particles.SpriteSize=&Multiply_Vector2D(
     A=$Particles.SpriteSize
     B=&SelectVector2DFromArray(Direct Array Index=$Particles.VisibilityTag, Vector 2D Selection Array=DataInterface() {propertyValues {FloatData [<1 1>, <1 1>, <1.25 2>]}})
    )) {isSetParameters=true, module=SetVariables_#3}
  ]
  particleUpdateScript [
   &ParticleState()
   &ScaleSpriteSize(Scale Sprite Size Mode=::NewEnumerator0) {module=ScaleSpriteSize001}
   &CameraOffset(Camera Offset Amount=&Multiply_Float(A=&MakeFloatFromVector2d(Vector 2D=$Particles.SpriteSize), B=0.33))
   &GravityForce(Gravity=&Multiply_VectorByFloat(Vector=<0 0 -980>, Float=1.5)) {enabled=false}
   &GravityForce(Gravity=&Multiply_VectorByFloat(Vector=<0 0 -980>, Float=3)) {module=GravityForce001}
   &Drag(
    Drag=&RandomRangeFloat(Minimum=2, Maximum=1)
    Rotational Drag=50
    Use Linear Drag=true
    Use Rotational Drag=false
   )
   &InheritSourceMovement(
    Applied Position Scale=&VectorFromFloat(Value=&FloatFromCurve(
      FloatCurve=DataInterface() {propertyValues {Curve {keys [{interpMode=RCIM_Cubic, tangentMode=RCTM_User, time=0, value=1}, {interpMode=RCIM_Cubic, tangentMode=RCTM_User, time=0.66, value=0}]}}}
      Scale Curve=$System.Inherit Source Movement
     ))
    ApplyForce=false
    ApplyPosition=true
    ApplyPositionFalloff=false
   ) {enabled=false}
   &Collision(
    Advanced Aging Rate=10
    Age Colliding Particles=false
    CPU Trace Vector Length Multiplier=3.5
    Collision Enabled=&Boolean_AndOperation(
     Bool1=&SetBoolByFloatComparison(A=$Particles.Age, B=0.1)
     Bool2=&Boolean_NotOperation(Bool=$Particles.Goo Collided Bool)
    )
    Enable Rest State=false
    Friction=0
    KillOnCollision=false
    Module.EnableMaxCollisionCount=false
    Module.Randomize Collision Normal Vector=false
    Particle Radius Scale=3
    Radius Calculation Type=::NewEnumerator3
    Restitution=-1
   )
   /SetVariables(Particles.Goo Collided Bool:NiagaraBool=$Particles.HasCollided) {isSetParameters=true, module=SetVariables_#4}
   /Niagara/Modules/Events/GenerateCollisionEvent(
    Delay Between Collision Events=0.025
    "Float to Send as Float 2 (RandomNormalizedFloat)":NiagaraFloat=&MakeFloatFromInt(Int=$Particles.VisibilityTag)
    Module.EventDelay=true
    Module.Use Velocity Threshold=true
   )
   &SolveForcesAndVelocity(Clamp Velocity=false, Limit Acceleration=false)
   &SpriteRotationRate(Rotation Rate=&FloatFromCurve(
     FloatCurve=DataInterface() {propertyValues {Curve {keys [<at 0 1>,<at 1 0>]}}}
     Scale Curve=&MakeCustomFloatFromBool(Bool=&RandomBool(), True Float=-90, False Float=90)
    ))
   &ScaleSpriteSize(Scale Factor=&Vector2DFromCurve(
     Vector2Curve=DataInterface() {propertyValues {
       XCurve {keys [{arriveTangent=0.5278, interpMode=RCIM_Cubic, time=0, value=0.5}, {arriveTangent=1.8113, tangentMode=RCTM_User, time=0.4, value=1}]}
       YCurve {keys [
         {arriveTangent=1.4008, interpMode=RCIM_Cubic, leaveTangent=1.4008, tangentMode=RCTM_User, time=0, value=0.2}
         {interpMode=RCIM_Cubic, tangentMode=RCTM_User, time=0.75, value=1}
        ]}
      }}
     ScaleCurve=<1 1.5>
    ))
   &ScaleSpriteSizeBySpeed(Max Scale Factor=<1.2 1.5>, Velocity Threshold=800)
   &ScaleColor(
    Scale Alpha=&FloatFromCurve(FloatCurve=DataInterface() {propertyValues {Curve {keys [{interpMode=RCIM_Cubic, tangentMode=RCTM_User, time=0.9, value=1}, {interpMode=RCIM_Cubic, tangentMode=RCTM_User, time=1, value=0}]}}})
    ScaleA=true
    ScaleRGB=true
   )
   &DynamicMaterialParameters(
    Index 0 Param 2=&FloatFromCurve(
     FloatCurve=DataInterface() {propertyValues {Curve {keys [{arriveTangent=-1.9579, interpMode=RCIM_Cubic, time=0.1, value=0.2}, {interpMode=RCIM_Cubic, tangentMode=RCTM_User, time=0.3, value=0}]}}}
     Scale Curve=0.5
    )
    Index 0 Param 4=&FloatFromCurve(
     FloatCurve=DataInterface() {propertyValues {Curve {keys [
         {arriveTangent=1.2665, interpMode=RCIM_Cubic, leaveTangent=1.2665, tangentMode=RCTM_User, time=0.1, value=0}
         {arriveTangent=1.0703, interpMode=RCIM_Cubic, leaveTangent=1.0703, tangentMode=RCTM_User, time=0.75, value=2}
        ]}}}
     Scale Curve=&RandomRangeFloat(Minimum=0.5, Maximum=0.8)
    )
    Module.Param0WriteEnabled=false
    Module.Param1WriteEnabled=true
    Module.Param2WriteEnabled=false
    Module.Param3WriteEnabled=true
   )
   /SetVariables(Particles.DispVel:Vector3f=$Particles.Initial.Velocity) {isSetParameters=true, module=SetVariables_#5}
   /SetVariables(Particles.DispVel:Vector3f=&Subtract_Position(
     A=$Particles.Initial.Position
     B=&AddVectorToPosition(Position=/Niagara/DynamicInputs/Helpers/SimulationPosition(), Vector=<0 0 -25>)
    )) {isSetParameters=true}
   &Color(Color=&HueShiftLinearColor(
     Hue=&MakeCustomFloatFromBool(Bool=$Particles.Goo Collided Bool, True Float=0.5, False Float=0)
     Saturation=1
     Value=1
     Linear Color=$Particles.Color
    )) {enabled=false}
   /SetVariables(Particles.Dynam 2:Vector4f=&MakeVector4(
     X=&RandomRangeFloat(Minimum=0, Maximum=1)
     Y=&FloatFromCurve(
      FloatCurve=DataInterface() {propertyValues {Curve {keys [{arriveTangent=-1.9579, interpMode=RCIM_Cubic, time=0.1, value=0.2}, {interpMode=RCIM_Cubic, tangentMode=RCTM_User, time=0.3, value=0}]}}}
      Scale Curve=0.5
     )
     Z=1
     W=&FloatFromCurve(FloatCurve=DataInterface() {propertyValues {Curve {keys [
          {
           arriveTangentWeight=0.4105
           interpMode=RCIM_Cubic
           leaveTangentWeight=0.3333
           tangentMode=RCTM_User
           tangentWeightMode=RCTWM_WeightedBoth
           time=0
           value=0.5593
          }
          {
           arriveTangentWeight=0.3333
           interpMode=RCIM_Cubic
           leaveTangentWeight=0.7342
           tangentMode=RCTM_User
           tangentWeightMode=RCTWM_WeightedBoth
           time=1
           value=0.8836
          }
         ]}}})
    )) {isSetParameters=true, module=SetVariables_#6}
  ]
  rendererClasses [&NiagaraSpriteRendererProperties]
  renderers [
   {rendererClass=&NiagaraSpriteRendererProperties}
   {rendererClass=&NiagaraSpriteRendererProperties, rendererIndex=1}
  ]
 }
 Splash_Hard_MedLow {
  emitterUpdateScript [
   &EmitterState(Life Cycle Mode=::NewEnumerator1, Loop Behavior=::NewEnumerator1, Loop Delay=0.1, Loop Duration=0.15, UseLoopDelay=true)
   &SpawnBurst_Instantaneous(Spawn Count=7, Use Spawn Probability=false, UseLoopCountLimit=false) {module=SpawnBurst_Instantaneous001}
   &SpawnRate(SpawnGroup=1, SpawnRate=120, Use Spawn Probability=true)
  ]
  particleSpawnScript [
   &InitializeParticle(
    Color=&RandomRangeLinearColor(Minimum=<rgba 1 1 1 1>, Maximum=<rgba 0.138 0.6337 0.1353 1>)
    Lifetime Min=2
    Lifetime Mode=::NewEnumerator1
    Position Offset Coordinate Space=::Local
    Sprite Rotation Mode=::NewEnumerator2
    Sprite Size Max=<100 120>
    Sprite Size Min=<80 150>
    Sprite Size Mode=::NewEnumerator2
    Sprite UV Mode=::NewEnumerator3
    UsePositionOffset=true
    Write Material Random=true
   )
   &PartitionParticles(Number Of Partitions=1, Particles Per Partition=7, Partition Mode:ENiagara_PartitionMode=::NewEnumerator1)
   &SubUVAnimation(End Frame=3, SubUV Animation Mode:ENiagara_SubUVLookupMode=::NewEnumerator2)
   &ShapeLocation(
    MaskSpawnGroup=false
    Override Seed=false
    Ring / Disc Distribution Mode:ENiagaraSphereDistributionMode=::NewEnumerator1
    Ring Radius=30
    Shape Primitive=::NewEnumerator4
    U Position=&MaskFloatBySpawnGroup(
     Masked Float=$Particles.PartitionParticles.Normalized Index In Partition
     Passthrough Float=&RandomRangeFloat(Minimum=0, Maximum=1)
    )
    UseOwnerScale=true
   )
   /SetVariables(Particles.VisibilityTag=&MakeCustomIntFromBool(
     Bool=&SetBoolByFloatComparison(A=&RandomRangeFloat(Minimum=0, Maximum=1), B=0.6)
     True INT=10
     False INT=&RandomRangeInt(Minimum=0, Maximum=1)
    )) {isSetParameters=true, module=SetVariables_#1}
   &AddVelocity(
    Module.Use Falloff Distance=false
    Origin Offset=<0 0 -80>
    Override Seed=false
    Velocity Mode=::NewEnumerator1
    Velocity Speed=&Multiply_Float(
     A=&RandomRangeFloat(Minimum=500, Maximum=600)
     B=&SelectFloatFromArray(Direct Array Index=$Particles.VisibilityTag, Float Selection Array=DataInterface() {propertyValues {FloatData [1.25,1.25,0.8]}})
    )
   ) {module=AddVelocity001}
   /SetVariables(Particles.SpriteSize=&Multiply_Vector2D(
     A=$Particles.SpriteSize
     B=&SelectVector2DFromArray(Direct Array Index=$Particles.VisibilityTag, Vector 2D Selection Array=DataInterface() {propertyValues {FloatData [<1 1>, <1 1>, <1.25 2>]}})
    )) {isSetParameters=true, module=SetVariables_#3}
  ]
  particleUpdateScript [
   &ParticleState()
   &CameraOffset(Camera Offset Amount=&Multiply_Float(A=&MakeFloatFromVector2d(Vector 2D=$Particles.SpriteSize), B=0.33))
   &GravityForce(Gravity=&Multiply_VectorByFloat(Vector=<0 0 -980>, Float=1.5))
   &Drag(
    Drag=&RandomRangeFloat(Minimum=2, Maximum=1)
    Rotational Drag=50
    Use Linear Drag=true
    Use Rotational Drag=false
   )
   &SolveForcesAndVelocity(Clamp Velocity=false, Limit Acceleration=false)
   &SpriteRotationRate(Rotation Rate=&FloatFromCurve(
     FloatCurve=DataInterface() {propertyValues {Curve {keys [<at 0 1>,<at 1 0>]}}}
     Scale Curve=&MakeCustomFloatFromBool(Bool=&RandomBool(), True Float=-90, False Float=90)
    ))
   &ScaleSpriteSize(Scale Factor=&Vector2DFromCurve(
     Vector2Curve=DataInterface() {propertyValues {
       XCurve {keys [{tangentMode=RCTM_User, time=0, value=0.4}, {tangentMode=RCTM_User, time=0.75, value=1}]}
       YCurve {keys [
         {arriveTangent=1.4008, interpMode=RCIM_Cubic, leaveTangent=1.4008, tangentMode=RCTM_User, time=0, value=0.1}
         {interpMode=RCIM_Cubic, tangentMode=RCTM_User, time=0.75, value=1}
        ]}
      }}
     ScaleCurve=<1 1.5>
    ))
   &ScaleSpriteSizeBySpeed(Max Scale Factor=<1.2 1.5>, Velocity Threshold=800)
   &ScaleColor(
    Scale Alpha=&FloatFromCurve(FloatCurve=DataInterface() {propertyValues {Curve {keys [{interpMode=RCIM_Cubic, tangentMode=RCTM_User, time=0.9, value=1}, {interpMode=RCIM_Cubic, tangentMode=RCTM_User, time=1, value=0}]}}})
    ScaleA=true
    ScaleRGB=true
   )
   &DynamicMaterialParameters(
    Index 0 Param 2=&FloatFromCurve(
     FloatCurve=DataInterface() {propertyValues {Curve {keys [{arriveTangent=-1.9579, interpMode=RCIM_Cubic, time=0.1, value=0.2}, {interpMode=RCIM_Cubic, tangentMode=RCTM_User, time=0.3, value=0}]}}}
     Scale Curve=0.5
    )
    Index 0 Param 4=&FloatFromCurve(
     FloatCurve=DataInterface() {propertyValues {Curve {keys [
         {arriveTangent=1.2665, interpMode=RCIM_Cubic, leaveTangent=1.2665, tangentMode=RCTM_User, time=0.1, value=0}
         {arriveTangent=1.0703, interpMode=RCIM_Cubic, leaveTangent=1.0703, tangentMode=RCTM_User, time=0.75, value=2}
        ]}}}
     Scale Curve=&RandomRangeFloat(Minimum=0.5, Maximum=0.8)
    )
    Module.Param0WriteEnabled=false
    Module.Param1WriteEnabled=true
    Module.Param2WriteEnabled=false
    Module.Param3WriteEnabled=true
   )
   /SetVariables(Particles.DispVel:Vector3f=$Particles.Initial.Velocity) {isSetParameters=true, module=SetVariables_#5}
   /SetVariables(Particles.Dynam 2:Vector4f=&MakeVector4(
     X=&RandomRangeFloat(Minimum=0, Maximum=1)
     Y=&FloatFromCurve(
      FloatCurve=DataInterface() {propertyValues {Curve {keys [{arriveTangent=-1.9579, interpMode=RCIM_Cubic, time=0.1, value=0.2}, {interpMode=RCIM_Cubic, tangentMode=RCTM_User, time=0.3, value=0}]}}}
      Scale Curve=0.5
     )
     Z=1
     W=&FloatFromCurve(FloatCurve=DataInterface() {propertyValues {Curve {keys [
          {
           arriveTangentWeight=0.4105
           interpMode=RCIM_Cubic
           leaveTangentWeight=0.3333
           tangentMode=RCTM_User
           tangentWeightMode=RCTWM_WeightedBoth
           time=0
           value=0.5593
          }
          {
           arriveTangentWeight=0.3333
           interpMode=RCIM_Cubic
           leaveTangentWeight=0.7342
           tangentMode=RCTM_User
           tangentWeightMode=RCTWM_WeightedBoth
           time=1
           value=0.8836
          }
         ]}}})
    )) {isSetParameters=true, module=SetVariables_#6}
  ]
  rendererClasses [&NiagaraSpriteRendererProperties]
  renderers [
   {rendererClass=&NiagaraSpriteRendererProperties}
   {rendererClass=&NiagaraSpriteRendererProperties, rendererIndex=1}
  ]
 }
 Splash_Invis {
  emitterUpdateScript [
   &EmitterState(Life Cycle Mode=::NewEnumerator1, Loop Behavior=::NewEnumerator1, Loop Delay=0.1, Loop Duration=0.08, UseLoopDelay=true)
   &SpawnBurst_Instantaneous(Spawn Count=7, Use Spawn Probability=false, UseLoopCountLimit=false) {enabled=false, module=SpawnBurst_Instantaneous001}
   &SpawnRate(SpawnGroup=1, SpawnRate=200, Use Spawn Probability=true)
  ]
  particleSpawnScript [
   &InitializeParticle(
    Color Mode=::NewEnumerator0
    Lifetime Min=2
    Lifetime Mode=::NewEnumerator1
    Position Offset Coordinate Space=::Local
    Sprite Rotation Mode=::NewEnumerator2
    Sprite Size Max=<100 120>
    Sprite Size Min=<80 150>
    Sprite Size Mode=::NewEnumerator2
    Sprite UV Mode=::NewEnumerator3
    UsePositionOffset=true
    Write Material Random=true
   )
   &PartitionParticles(Number Of Partitions=1, Particles Per Partition=7, Partition Mode:ENiagara_PartitionMode=::NewEnumerator1)
   &SubUVAnimation(End Frame=3, SubUV Animation Mode:ENiagara_SubUVLookupMode=::NewEnumerator2)
   &ShapeLocation(
    MaskSpawnGroup=false
    Override Seed=false
    Ring / Disc Distribution Mode:ENiagaraSphereDistributionMode=::NewEnumerator1
    Ring Radius=30
    Shape Primitive=::NewEnumerator4
    U Position=&MaskFloatBySpawnGroup(
     Masked Float=$Particles.PartitionParticles.Normalized Index In Partition
     Passthrough Float=&RandomRangeFloat(Minimum=0, Maximum=1)
    )
    UseOwnerScale=true
   )
   /SetVariables(Particles.VisibilityTag=&MakeCustomIntFromBool(
     Bool=&SetBoolByFloatComparison(A=&RandomRangeFloat(Minimum=0, Maximum=1), B=0.6)
     True INT=10
     False INT=&RandomRangeInt(Minimum=0, Maximum=1)
    )) {isSetParameters=true, module=SetVariables_#1}
   &AddVelocity(
    Module.Use Falloff Distance=false
    Origin Offset=<0 0 -80>
    Override Seed=false
    Velocity Mode=::NewEnumerator1
    Velocity Speed=&Multiply_Float(
     A=&RandomRangeFloat(Minimum=500, Maximum=600)
     B=&SelectFloatFromArray(Direct Array Index=$Particles.VisibilityTag, Float Selection Array=DataInterface() {propertyValues {FloatData [1.25,1.25,0.8]}})
    )
   ) {enabled=false, module=AddVelocity001}
   &AddVelocity(
    Module.Use Falloff Distance=false
    Origin Offset=<0 0 -80>
    Override Seed=false
    Velocity Mode=::NewEnumerator1
    Velocity Speed=&Multiply_Float(A=&RandomRangeFloat(Minimum=500, Maximum=600), B=1.5)
   )
   /SetVariables() {isSetParameters=true, module=SetVariables_#2}
   /SetVariables(Particles.SpriteSize=&Multiply_Vector2D(
     A=$Particles.SpriteSize
     B=&SelectVector2DFromArray(Direct Array Index=$Particles.VisibilityTag, Vector 2D Selection Array=DataInterface() {propertyValues {FloatData [<1 1>, <1 1>, <1.25 2>]}})
    )) {isSetParameters=true, module=SetVariables_#3}
  ]
  particleUpdateScript [
   &ParticleState()
   &CameraOffset(Camera Offset Amount=&Multiply_Float(A=&MakeFloatFromVector2d(Vector 2D=$Particles.SpriteSize), B=0.33))
   &GravityForce(Gravity=&Multiply_VectorByFloat(Vector=<0 0 -980>, Float=3))
   &Drag(
    Drag=&RandomRangeFloat(Minimum=2, Maximum=1)
    Rotational Drag=50
    Use Linear Drag=true
    Use Rotational Drag=false
   )
   &InheritSourceMovement(
    Applied Position Scale=&VectorFromFloat(Value=&FloatFromCurve(
      FloatCurve=DataInterface() {propertyValues {Curve {keys [{interpMode=RCIM_Cubic, tangentMode=RCTM_User, time=0, value=1}, {interpMode=RCIM_Cubic, tangentMode=RCTM_User, time=0.66, value=0}]}}}
      Scale Curve=$System.Inherit Source Movement
     ))
    ApplyForce=false
    ApplyPosition=true
    ApplyPositionFalloff=false
   ) {enabled=false}
   &Collision(
    Advanced Aging Rate=10
    Age Colliding Particles=false
    CPU Trace Vector Length Multiplier=50
    Collision Enabled=&Boolean_AndOperation(
     Bool1=&SetBoolByFloatComparison(A=$Particles.Age, B=0.1)
     Bool2=&Boolean_NotOperation(Bool=$Particles.Goo Collided Bool)
    )
    Enable Rest State=false
    Friction=0
    KillOnCollision=false
    Module.EnableMaxCollisionCount=false
    Module.Randomize Collision Normal Vector=false
    Radius Calculation Type=::NewEnumerator3
    Restitution=-1
   )
   /SetVariables(Particles.Goo Collided Bool:NiagaraBool=$Particles.HasCollided) {isSetParameters=true, module=SetVariables_#4}
   &GenerateCollisionEvent(
    "Float to Send as Float 2 (RandomNormalizedFloat)":NiagaraFloat=&MakeFloatFromInt(Int=$Particles.VisibilityTag)
    Module.EventDelay=true
    Module.Use Velocity Threshold=true
   )
   &SolveForcesAndVelocity(Clamp Velocity=false, Limit Acceleration=false)
   &SpriteRotationRate(Rotation Rate=&FloatFromCurve(
     FloatCurve=DataInterface() {propertyValues {Curve {keys [<at 0 1>,<at 1 0>]}}}
     Scale Curve=&MakeCustomFloatFromBool(Bool=&RandomBool(), True Float=-90, False Float=90)
    ))
   &ScaleSpriteSize(Scale Factor=&Vector2DFromCurve(
     Vector2Curve=DataInterface() {propertyValues {
       XCurve {keys [{tangentMode=RCTM_User, time=0, value=0.4}, {tangentMode=RCTM_User, time=0.75, value=1}]}
       YCurve {keys [
         {arriveTangent=1.4008, interpMode=RCIM_Cubic, leaveTangent=1.4008, tangentMode=RCTM_User, time=0, value=0.1}
         {interpMode=RCIM_Cubic, tangentMode=RCTM_User, time=0.75, value=1}
        ]}
      }}
     ScaleCurve=<1 1.5>
    ))
   &ScaleSpriteSizeBySpeed(Max Scale Factor=<1.2 1.5>, Velocity Threshold=800)
   &ScaleColor(
    Scale Alpha=&FloatFromCurve(FloatCurve=DataInterface() {propertyValues {Curve {keys [{interpMode=RCIM_Cubic, tangentMode=RCTM_User, time=0.9, value=1}, {interpMode=RCIM_Cubic, tangentMode=RCTM_User, time=1, value=0}]}}})
    ScaleA=true
    ScaleRGB=true
   ) {enabled=false}
   &DynamicMaterialParameters(
    Index 0 Param 2=&FloatFromCurve(
     FloatCurve=DataInterface() {propertyValues {Curve {keys [{arriveTangent=-1.9579, interpMode=RCIM_Cubic, time=0.1, value=0.2}, {interpMode=RCIM_Cubic, tangentMode=RCTM_User, time=0.3, value=0}]}}}
     Scale Curve=0.5
    )
    Index 0 Param 4=&FloatFromCurve(
     FloatCurve=DataInterface() {propertyValues {Curve {keys [
         {arriveTangent=1.2665, interpMode=RCIM_Cubic, leaveTangent=1.2665, tangentMode=RCTM_User, time=0.1, value=0}
         {arriveTangent=1.0703, interpMode=RCIM_Cubic, leaveTangent=1.0703, tangentMode=RCTM_User, time=0.75, value=2}
        ]}}}
     Scale Curve=&RandomRangeFloat(Minimum=0.5, Maximum=0.8)
    )
    Module.Param0WriteEnabled=false
    Module.Param1WriteEnabled=true
    Module.Param2WriteEnabled=false
    Module.Param3WriteEnabled=true
   )
   /SetVariables(Particles.DispVel:Vector3f=$Particles.Initial.Velocity) {isSetParameters=true, module=SetVariables_#5}
   &Color(Color=&HueShiftLinearColor(
     Hue=&MakeCustomFloatFromBool(Bool=$Particles.Goo Collided Bool, True Float=0.5, False Float=0)
     Saturation=1
     Value=1
     Linear Color=$Particles.Color
    )) {enabled=false}
   /SetVariables(Particles.Dynam 2:Vector4f=&MakeVector4(
     X=&RandomRangeFloat(Minimum=0, Maximum=1)
     Y=&FloatFromCurve(
      FloatCurve=DataInterface() {propertyValues {Curve {keys [{arriveTangent=-1.9579, interpMode=RCIM_Cubic, time=0.1, value=0.2}, {interpMode=RCIM_Cubic, tangentMode=RCTM_User, time=0.3, value=0}]}}}
      Scale Curve=0.5
     )
     Z=1
     W=&FloatFromCurve(FloatCurve=DataInterface() {propertyValues {Curve {keys [
          {
           arriveTangentWeight=0.4105
           interpMode=RCIM_Cubic
           leaveTangentWeight=0.3333
           tangentMode=RCTM_User
           tangentWeightMode=RCTWM_WeightedBoth
           time=0
           value=0.5593
          }
          {
           arriveTangentWeight=0.3333
           interpMode=RCIM_Cubic
           leaveTangentWeight=0.7342
           tangentMode=RCTM_User
           tangentWeightMode=RCTWM_WeightedBoth
           time=1
           value=0.8836
          }
         ]}}})
    )) {isSetParameters=true, module=SetVariables_#6}
  ]
  rendererClasses [&NiagaraSpriteRendererProperties]
  renderers [{rendererClass=&NiagaraSpriteRendererProperties}]
 }
 Splash_Smooth {
  emitterUpdateScript [
   &EmitterState()
   &SpawnBurst_Instantaneous(Spawn Count=3, Use Spawn Probability=false, UseLoopCountLimit=false) {module=SpawnBurst_Instantaneous001}
   &SpawnRate(
    SpawnRate=&FloatFromCurve(
     FloatCurve=DataInterface() {propertyValues {Curve {keys [{interpMode=RCIM_Cubic, tangentMode=RCTM_User, time=0, value=1}, {interpMode=RCIM_Cubic, tangentMode=RCTM_User, time=1, value=0}]}}}
     CurveIndex=$Emitter.NormalizedLoopAge
     Scale Curve=10
    )
    Use Spawn Probability=false
   ) {enabled=false}
   &SpawnBurst_Instantaneous(Spawn Count=24, Spawn Time=0.1, Use Spawn Probability=false) {enabled=false}
   &SpawnBurst_Instantaneous(Spawn Count=12, Spawn Time=0.15, Use Spawn Probability=false) {enabled=false, module=SpawnBurst_Instantaneous002}
  ]
  particleSpawnScript [
   &FN_SetSpawnGroup() {enabled=false}
   /Niagara/Modules/Spawn/Initialization/InitializeParticle(
    Color=&RandomRangeLinearColor(Minimum=<rgba 3.2063 4.2281 5 2>, Maximum=<rgba 0.2 0.25 0.3 0>)
    Lifetime=&RandomRangeFloat(Minimum=0.8, Maximum=1.4)
    Mass=&RandomRangeFloat(Minimum=1, Maximum=2)
    Module.Write Color=true
    Module.Write Lifetime=true
    Module.Write Mass=true
    Module.Write Material Random=true
    Module.Write Position=false
    Module.Write Scale=false
    Module.Write SpriteRotation=true
    Module.Write SpriteSize=true
    Position=/Niagara/Functions/Localspace/SimulationPosition()
    Sprite Rotation=&UniformRangedFloat(Minimum=0, Maximum=360)
    Sprite Size=&Vector2DFromFloat(Value=&UniformRangedFloat(Minimum=10, Maximum=30))
   ) {enabled=false}
   /Niagara/Modules/Spawn/Initialization/V2/InitializeParticle(
    Color=&RandomRangeLinearColor(Minimum=<rgba 0.122 3 0 1>, Maximum=<rgba 0.1729 2 0.0022 1>)
    Lifetime=&RandomRangeFloat(Minimum=1, Maximum=1)
    Mass=&RandomRangeFloat(Minimum=1, Maximum=2)
    Mass Mode=::NewEnumerator1
    Sprite Size Max=<22 50>
    Sprite Size Min=<20 30>
    Sprite Size Mode=::NewEnumerator2
    Sprite UV Mode=::NewEnumerator3
    UsePositionOffset=true
    Write Material Random=true
   ) {module=InitializeParticle001}
   &ShapeLocation(
    MaskSpawnGroup=false
    Offset Mode=::NewEnumerator0
    Override Seed=false
    Ring Radius=30
    Shape Primitive=::NewEnumerator4
    UseOwnerScale=true
   )
   &AddVelocity(
    Module.Use Falloff Distance=false
    Origin Offset=&Add_Vector(A=<0 0 -50>, B=$System.Location Offset)
    Override Seed=false
    Velocity Mode=::NewEnumerator1
    Velocity Speed=&Multiply_Float(A=&RandomRangeFloat(Minimum=250, Maximum=500), B=0.75)
   )
   &AddVelocityFromPoint(
    Module.Use Falloff Distance=false
    Origin Offset=&SelectVectorFromArray(
     Direct Array Index=$Particles.SpawnGroup
     Vector Selection Array=DataInterface() {propertyValues {FloatData [<-5 0 -12>, <5 0 -12>, <0 -5 -12>, <0 5 -12>]}}
    )
    Origin Offset Coordinate Space=::World
    Velocity Strength=&UniformRangedFloat(Minimum=450, Maximum=800)
   ) {enabled=false}
  ]
  particleUpdateScript [
   &ParticleState()
   &Drag(Drag=2, Rotational Drag=0, Use Linear Drag=true, Use Rotational Drag=true)
   &AccelerationForce(Acceleration=&Multiply_VectorByFloat(Vector=&UniformRangedVector(Minimum=<0 0 -300>, Maximum=<0 0 -500>), Float=1.5))
   &SolveForcesAndVelocity(Clamp Velocity=false, Limit Acceleration=false)
   &ScaleSpriteSize(Scale Factor=&Vector2DFromFloat(Value=&FloatFromCurve(
      FloatCurve=DataInterface() {propertyValues {Curve {keys [
          {arriveTangent=12.5197, interpMode=RCIM_Cubic, leaveTangent=12.5197, tangentMode=RCTM_User, time=0, value=0.4376}
          {arriveTangent=0.8929, interpMode=RCIM_Cubic, leaveTangent=0.8929, tangentMode=RCTM_User, time=1, value=5}
         ]}}}
      Scale Curve=&RandomRangeFloat(Minimum=1, Maximum=0.5)
     )))
   &ScaleColor(
    Scale Alpha=&FloatFromCurve(FloatCurve=DataInterface() {propertyValues {Curve {keys [{interpMode=RCIM_Cubic, tangentMode=RCTM_User, time=0.9, value=1}, {tangentMode=RCTM_User, time=1, value=0}]}}})
    ScaleA=true
    ScaleRGB=true
   )
   &DynamicMaterialParameters(
    Index 0 Param 1=&FloatFromCurve(FloatCurve=DataInterface() {propertyValues {Curve {keys [{interpMode=RCIM_Cubic, tangentMode=RCTM_User, time=0, value=1}, {interpMode=RCIM_Cubic, tangentMode=RCTM_User, time=1, value=0}]}}})
    Index 0 Param 2=&FloatFromCurve(FloatCurve=DataInterface() {propertyValues {Curve {keys [<at -0.0086 -1.3001>,<at 0.9945 1.9153>]}}})
    Index 0 Param 3=0.5
    Module.Param0WriteEnabled=true
    Module.Param1WriteEnabled=false
    Module.Param2WriteEnabled=true
    Module.Param3WriteEnabled=true
   ) {module=DynamicMaterialParameters001}
   &SpriteRotationRate(Rotation Rate=&Lerp_Float(A=-150, B=150, Alpha=$Particles.Initial.MaterialRandom)) {enabled=false}
   /SetVariables(Particles.DispVel:Vector3f=$Particles.Initial.Velocity) {isSetParameters=true, module=SetVariables_#5}
  ]
  rendererClasses [&NiagaraSpriteRendererProperties]
  renderers [{rendererClass=&NiagaraSpriteRendererProperties}]
 }
 Stickey {
  emitterSpawnScript [/SetVariables(Emitter.Mat Rand:NiagaraFloat=&RandomRangeFloat(Minimum=0, Maximum=1)) {isSetParameters=true}]
  emitterUpdateScript [&EmitterState()]
  particleSpawnScript [
   &InitializeParticle(
    Color Maximum=<rgba 0.138 0.6337 0.1353 1>
    Color Minimum=<rgba 1 1 1 1>
    Color Mode=::NewEnumerator2
    Lifetime Min=2.5
    Lifetime Mode=::NewEnumerator1
    Position Offset=<0 0 1>
    Position Offset Coordinate Space=::Local
    Sprite Size Max=<120 150>
    Sprite Size Min=<80 100>
    Sprite Size Mode=::NewEnumerator2
    Sprite UV Mode=::NewEnumerator3
    UsePositionOffset=true
    Write Material Random=true
   )
   &SubUVAnimation(
    Random Start Frame=true
    "Sprite Renderer "=DataInterface() {propertyValues {SpriteRenderer=&self:Stickey_0.NiagaraSpriteRendererProperties_0}}
    UseEndFrame=false
    UseStartFrame=false
   ) {enabled=false}
   &SubUVAnimation(End Frame=3, SubUV Animation Mode:ENiagara_SubUVLookupMode=::NewEnumerator2) {module=SubUVAnimation001}
   &InitialMeshOrientation(Use Rotation Vector=false)
   /SetVariables(Particles.DecalFade=0.5) {isSetParameters=true}
  ]
  particleUpdateScript [
   &ParticleState()
   &ScaleColor(
    Scale Alpha=&FloatFromCurve(
     FloatCurve=DataInterface() {propertyValues {Curve {keys [
         {arriveTangentWeight=0.2357, interpMode=RCIM_Cubic, leaveTangentWeight=0.2357, tangentMode=RCTM_User, time=0.9, value=1}
         {tangentMode=RCTM_User, time=1, value=0}
        ]}}}
     Scale Curve=25
    )
    ScaleA=true
    ScaleRGB=true
   )
   &DynamicMaterialParameters(
    Index 0 Param 2=&FloatFromCurve(
     FloatCurve=DataInterface() {propertyValues {Curve {keys [{arriveTangent=-1.9579, interpMode=RCIM_Cubic, time=0.2, value=0}, {interpMode=RCIM_Cubic, tangentMode=RCTM_User, time=0.45, value=0}]}}}
     Scale Curve=0.5
    )
    Index 0 Param 4=&FloatFromCurve(
     FloatCurve=DataInterface() {propertyValues {Curve {keys [
         {arriveTangent=-25.3206, interpMode=RCIM_Cubic, leaveTangent=-25.3206, tangentMode=RCTM_User, time=0, value=2}
         {arriveTangentWeight=0.0667, interpMode=RCIM_Cubic, leaveTangentWeight=0.2667, tangentMode=RCTM_User, time=0.15, value=0.5}
         {arriveTangentWeight=0.359, interpMode=RCIM_Cubic, leaveTangentWeight=0.359, tangentMode=RCTM_User, time=0.5, value=0.5}
         {arriveTangent=4.9537, interpMode=RCIM_Cubic, leaveTangent=4.9537, tangentMode=RCTM_User, time=1, value=2}
        ]}}}
     Scale Curve=&RandomRangeFloat(Minimum=0.5, Maximum=0.8)
    )
    Module.Param0WriteEnabled=false
    Module.Param1WriteEnabled=true
    Module.Param2WriteEnabled=false
    Module.Param3WriteEnabled=true
   )
   /SetVariables(Particles.Dynam 2:Vector4f=&MakeVector4(
     X=&RandomRangeFloat(Minimum=0, Maximum=1)
     Y=&FloatFromCurve(
      FloatCurve=DataInterface() {propertyValues {Curve {keys [{arriveTangent=-1.9579, interpMode=RCIM_Cubic, time=0.1, value=0.2}, {interpMode=RCIM_Cubic, tangentMode=RCTM_User, time=0.3, value=0}]}}}
      Scale Curve=0.5
     )
     Z=0
     W=&FloatFromCurve(FloatCurve=DataInterface() {propertyValues {Curve {keys [
          {
           arriveTangent=-2.2031
           arriveTangentWeight=0.0864
           interpMode=RCIM_Cubic
           leaveTangent=-2.2031
           leaveTangentWeight=0.0864
           tangentMode=RCTM_User
           tangentWeightMode=RCTWM_WeightedBoth
           time=0
           value=1
          }
          {arriveTangentWeight=0.05, interpMode=RCIM_Cubic, leaveTangentWeight=0.2833, tangentMode=RCTM_User, time=0.2, value=0.7}
          {arriveTangentWeight=0.1167, interpMode=RCIM_Cubic, leaveTangentWeight=0.1667, tangentMode=RCTM_User, time=0.5, value=0.7}
          {arriveTangent=4.9537, arriveTangentWeight=0.1667, interpMode=RCIM_Cubic, leaveTangent=4.9537, tangentMode=RCTM_User, time=1, value=2}
         ]}}})
    )) {isSetParameters=true}
  ]
  rendererClasses [&NiagaraSpriteRendererProperties]
  renderers [
   {rendererClass=&NiagaraSpriteRendererProperties}
   {rendererClass=&NiagaraSpriteRendererProperties, rendererIndex=1}
  ]
 }
 buffetSplash {
  emitterUpdateScript [
   &EmitterState()
   &SpawnRate(SpawnRate=10, Use Spawn Probability=false) {enabled=false}
   &SpawnBurst_Instantaneous(Spawn Count=&RandomRangeInt(Minimum=4, Maximum=7), Spawn Time=0.05, Use Spawn Probability=false) {module=SpawnBurst_Instantaneous001}
   &SpawnBurst_Instantaneous(Spawn Count=&RandomRangeInt(Minimum=4, Maximum=7), Use Spawn Probability=false)
  ]
  particleSpawnScript [
   &InitializeParticle(
    Color=&RandomRangeLinearColor(Minimum=<rgba 0 1.5 0.0116 1>, Maximum=<rgba 0 5 0.0388 1>)
    Lifetime Max=1.25
    Lifetime Min=0.6
    Lifetime Mode=::NewEnumerator1
    Sprite Rotation Mode=::NewEnumerator2
    Sprite Size Max=<300 930>
    Sprite Size Min=<262.5 450>
    Sprite Size Mode=::NewEnumerator2
    Sprite UV Mode=::NewEnumerator3
    UsePositionOffset=false
    Write Material Random=true
   ) {enabled=false}
   &InitializeParticle(
    Color=&RandomRangeLinearColor(Minimum=<rgba 1 1 1 1>, Maximum=<rgba 0.138 0.6337 0.1353 1>)
    Lifetime Max=1.25
    Lifetime Min=0.85
    Lifetime Mode=::NewEnumerator1
    Sprite Size Mode=::NewEnumerator3
    Sprite UV Mode=::NewEnumerator3
    Uniform Sprite Size=450
    UsePositionOffset=false
    Write Material Random=true
   ) {module=InitializeParticle001}
   &AddVelocityInCone(
    Cone Angle=75
    Cone Axis=<0 0 1>
    Module.Use Velocity Falloff On Cone Axis=false
    Velocity Strength=&Multiply_Float(A=&RandomRangeFloat(Minimum=1400, Maximum=2100), B=0.3)
   ) {enabled=false}
   &AddVelocityInCone(
    Cone Angle=90
    Cone Axis=<0 0 1>
    Module.Use Velocity Falloff On Cone Axis=false
    Velocity Strength=&Multiply_Float(A=&RandomRangeFloat(Minimum=1400, Maximum=2100), B=0.4)
   ) {enabled=false, module=AddVelocityInCone001}
   &DynamicMaterialParameters(
    Index 0 Param 1=0.5
    Index 0 Param 2=0.5
    Index 0 Param 3=25
    Index 0 Param 4=&RandomRangeFloat(Minimum=0, Maximum=10)
    Module.Param0WriteEnabled=true
    Module.Param1WriteEnabled=true
    Module.Param2WriteEnabled=true
    Module.Param3WriteEnabled=true
   ) {enabled=false, module=DynamicMaterialParameters001}
   &AddVelocity(
    Cone Angle=30
    Cone Axis=<0 0 1>
    Inner Cone Angle=35
    Module.UseConeFalloff=false
    Module.UseVelDistribution=false
    Override Seed=false
    Velocity Mode=::NewEnumerator2
    Velocity Speed=&Multiply_Float(A=&RandomRangeFloat(Minimum=1400, Maximum=2100), B=0.5)
   )
  ]
  particleUpdateScript [
   &ParticleState()
   &ScaleSpriteSize(Scale Factor=<0.35 0.35>, Scale Sprite Size Mode=::NewEnumerator2) {module=ScaleSpriteSize001}
   &SpriteRotationRate(Rotation Rate=&FloatFromCurve(
     FloatCurve=DataInterface() {propertyValues {Curve {keys [<at 0.35 0>,<at 1 1>]}}}
     Scale Curve=&Multiply_Float(
      A=&RandomRangeFloat(Minimum=280, Maximum=230)
      B=&MakeCustomFloatFromBool(Bool=&RandomBool(), True Float=1, False Float=-1)
     )
    ))
   &ScaleSpriteSize(Scale Factor=&MakeVector2D(
     X=&FloatFromCurve(
      FloatCurve=DataInterface() {propertyValues {Curve {keys [
          {arriveTangent=-0.1528, interpMode=RCIM_Cubic, leaveTangent=-0.1528, tangentMode=RCTM_User, time=0, value=0.3}
          {arriveTangent=0.9213, interpMode=RCIM_Cubic, leaveTangent=0.9213, tangentMode=RCTM_User, time=0.25, value=0.5}
          {arriveTangent=0.5099, interpMode=RCIM_Cubic, leaveTangent=0.5099, tangentMode=RCTM_User, time=1, value=1}
         ]}}}
      Scale Curve=0.2
     )
     Y=&FloatFromCurve(
      FloatCurve=DataInterface() {propertyValues {Curve {keys [
          {arriveTangent=-0.1528, interpMode=RCIM_Cubic, leaveTangent=-0.1528, tangentMode=RCTM_User, time=0, value=0}
          {arriveTangent=0.9213, interpMode=RCIM_Cubic, leaveTangent=0.9213, tangentMode=RCTM_User, time=0.25, value=0.5}
          {arriveTangent=0.5099, interpMode=RCIM_Cubic, leaveTangent=0.5099, tangentMode=RCTM_User, time=1, value=1}
         ]}}}
      Scale Curve=0.2
     )
    )) {enabled=false}
   &ScaleSpriteSize(Scale Factor=&MakeVector2D(
     X=&FloatFromCurve(
      FloatCurve=DataInterface() {propertyValues {Curve {keys [
          {arriveTangent=-0.1528, interpMode=RCIM_Cubic, time=0, value=0.3}
          {arriveTangent=3.3554, interpMode=RCIM_Cubic, leaveTangent=3.3554, tangentMode=RCTM_User, time=0.6, value=1.35}
         ]}}}
      Scale Curve=0.2
     )
     Y=&FloatFromCurve(
      FloatCurve=DataInterface() {propertyValues {Curve {keys [
          {
           arriveTangent=-2.2494
           interpMode=RCIM_Cubic
           leaveTangent=-2.2494
           leaveTangentWeight=0.8206
           tangentMode=RCTM_User
           tangentWeightMode=RCTWM_WeightedBoth
           time=0
           value=3
          }
          <at 0.5 1>
         ]}}}
      Scale Curve=0.2
     )
    )) {module=ScaleSpriteSize002}
   &GravityForce(Gravity=&Multiply_VectorByFloat(Vector=&RandomRangeVector(Minimum=<0 0 -1000>, Maximum=<0 0 -1250>), Float=0.35)) {enabled=false}
   &GravityForce(Gravity=&Multiply_VectorByFloat(Vector=&RandomRangeVector(Minimum=<0 0 -1000>, Maximum=<0 0 -1250>), Float=1.4)) {module=GravityForce001}
   &Drag(
    Drag=&Multiply_Float(A=&RandomRangeFloat(Minimum=0.75, Maximum=1.75), B=1.25)
    Use Linear Drag=true
    Use Rotational Drag=false
   )
   &SolveForcesAndVelocity(Clamp Velocity=false, Limit Acceleration=false)
   &ScaleColor(
    Scale Alpha=&FloatFromCurve(FloatCurve=DataInterface() {propertyValues {Curve {keys [{tangentMode=RCTM_User, time=0.9, value=1}, {tangentMode=RCTM_User, time=1, value=0}]}}})
    ScaleA=true
    ScaleRGB=false
   )
   &DynamicMaterialParameters(
    Index 0 Param 1=&RandomRangeFloat(Minimum=0, Maximum=1)
    Index 0 Param 2=&FloatFromCurve(
     FloatCurve=DataInterface() {propertyValues {Curve {keys [{arriveTangent=-1.9579, interpMode=RCIM_Cubic, time=0.1, value=0.2}, {interpMode=RCIM_Cubic, tangentMode=RCTM_User, time=0.3, value=0}]}}}
     Scale Curve=0.5
    )
    Index 0 Param 3=1
    Index 0 Param 4=&FloatFromCurve(FloatCurve=DataInterface() {propertyValues {Curve {keys [
         {
          arriveTangent=0.5721
          arriveTangentWeight=0.3605
          interpMode=RCIM_Cubic
          leaveTangent=0.5721
          leaveTangentWeight=0.3605
          tangentMode=RCTM_User
          tangentWeightMode=RCTWM_WeightedBoth
          time=0
          value=0
         }
         {arriveTangent=0.7084, tangentMode=RCTM_User, time=0.55, value=1.35}
        ]}}})
    Module.Index0Param0Write=true
    Module.Index0Param1Write=true
    Module.Index0Param2Write=true
    Module.Index0Param3Write=true
   )
   /SetVariables(Particles.DispVel:Vector3f=$Particles.Initial.Velocity) {isSetParameters=true}
   &KillParticles(Kill Particles=&SetBoolByFloatComparison(A=$Particles.NormalizedAge, B=0.6))
  ]
  rendererClasses [&NiagaraSpriteRendererProperties]
  renderers [{rendererClass=&NiagaraSpriteRendererProperties}]
 }
}
summary {
 systemName=NS_QuietPeanuts_EmotePropSpawn
 emitters [
  {bEnabled=false, emitterName=Splash_01, rendererClasses [&NiagaraSpriteRendererProperties]}
  {bEnabled=false, emitterName=Splash_02, rendererClasses [&NiagaraSpriteRendererProperties]}
  {emitterName=Splash_Smooth, rendererClasses [&NiagaraSpriteRendererProperties]}
  {emitterName=Splash_Hard, rendererClasses [&NiagaraSpriteRendererProperties]}
  {emitterName=AirGlow, rendererClasses [&NiagaraSpriteRendererProperties, &NiagaraMeshRendererProperties]}
  {emitterName=Smoke, rendererClasses [&NiagaraSpriteRendererProperties]}
  {emitterName=Sparkles, rendererClasses [&NiagaraSpriteRendererProperties]}
  {emitterName=Stickey, rendererClasses [&NiagaraSpriteRendererProperties]}
  {emitterName=Light, rendererClasses [&NiagaraLightRendererProperties]}
  {emitterName=buffetSplash, rendererClasses [&NiagaraSpriteRendererProperties]}
  {emitterName=Splash_Hard_MedLow, rendererClasses [&NiagaraSpriteRendererProperties]}
  {emitterName=Splash_Invis, rendererClasses [&NiagaraSpriteRendererProperties]}
  {emitterName=BuffSplash_Mesh, rendererClasses [&NiagaraMeshRendererProperties]}
 ]
 userVariables [
  User.bIsFrontEnd {type=NiagaraBool}
  User.bIsFrontEndPreview {type=NiagaraBool}
 ]
}
